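# Placeholder container material for the Macaron passthrough vehicle cosmetic mesh, flagged for skeletal mesh and mesh deformer use, apparently for Mutable.
# It declares texture, paint, rim, specular, WPO animation and sparkle parameters, including a sparkle mask channel selector.
# One long Multiply chain ties every parameter together but reaches no material output, so the graph holds parameters rather than implementing shading or animation.
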Material /VehicleCosmetics/PassthroughMeshes/Macaron/M_Macaron_MutableContainer
graph {
 properties {bAutomaticallySetUsageInEditor=false, bUsedWithMeshDeformer=true, bUsedWithSkeletalMesh=true, usedShadingModels=Default Lit}
 nodes [
  n1 = TextureSampleParameter2D() {
   group=1_Input Textures
   parameterName=BC
   samplerSource=SSM_Wrap_WorldGroupSettings
   sortPriority=0
   texture=/VehicleCosmetics/Textures/T_Black
  }
  n2 = TextureSampleParameter2D() {
   group=1_Input Textures
   parameterName=Mask
   samplerSource=SSM_Wrap_WorldGroupSettings
   samplerType=SAMPLERTYPE_Masks
   sortPriority=1
   texture=/VehicleCosmetics/Textures/T_Black_Masks
   channelNames {a=PaintMultiply, b=PaintMask, g=Rough, r=Metallic}
  }
  n3 = Multiply(A=@n1.RGB, B=@n2.RGB)
  n4 = TextureSampleParameter2D() {
   group=1_Input Textures
   parameterName=Normal
   samplerSource=SSM_Wrap_WorldGroupSettings
   samplerType=SAMPLERTYPE_Normal
   sortPriority=2
   texture=/VehicleCosmetics/Textures/T_BlankNormal
  }
  n5 = VectorParameter() {group=2_WPO Animation, parameterName=TranslationAmount, sortPriority=1, defaultValue=<rgba 0.25 0.25 0.25 1>}
  n6 = VectorParameter() {
   desc="If \"Painted\" is set to 1, this color will multiply over the \r\nBC according to mask. Does nothing if \"Painted\" is set to 0."
   group=2_Material Properties
   parameterName=PaintColor
   sortPriority=2
   defaultValue=<rgba 0.4635 0 0.0126 1>
  }
  n7 = Multiply(A=Multiply(A=Multiply(A=@n3, B=@n4.RGB), B=@n5.RGB), B=@n6.RGB)
  n8 = ScalarParameter() {defaultValue=0.5, group=2_Material Properties, parameterName=Specular, sortPriority=5}
  n9 = ScalarParameter() {
   desc="0==False, 1==True: Set at 1 if Painted Color overrides \r\nBC, 0 for not Painted. Auto-rounds to 0 or 1."
   group=2_Material Properties
   parameterName=Painted
   sortPriority=1
  }
  n10 = Multiply(
   A=Multiply(A=Multiply(A=@n7, B=@n8), B=@n9)
   B=ScalarParameter() {defaultValue=1, group=2_WPO Animation, parameterName=Animation, sortPriority=2}
  )
  n11 = ScalarParameter() {defaultValue=2, group=2_Material Properties, parameterName=RimSharpness}
  n12 = VectorParameter() {
   desc="Higher tiling values == denser sparkles."
   group=05_Shader Effects
   parameterName=SparkleTiling
   sortPriority=41
   defaultValue=<rgba 4 4 0 1>
  }
  n13 = Multiply(
   A=Multiply(A=Multiply(A=@n10, B=@n11), B=ScalarParameter() {defaultValue=2, group=2_Material Properties, parameterName=RimValue})
   B=@n12.RGB
  )
  n14 = ChannelMaskParameter(None=@n2.RGB) {group=05_Shader Effects, parameterName=SparkleMaskChannelSelect, sortPriority=46}
  n15 = ScalarParameter() {defaultValue=4096, desc="Lower values == fewer standout sparkles.", group=05_Shader Effects, parameterName=SparkleIntensity, sortPriority=49}
  Multiply(B=Multiply(A=Multiply(A=@n13, B=@n14.R), B=@n15))
 ]
}
