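# Physics asset for the Macaron skeletal mesh in vehicle cosmetics, named as a Mutable container variant.
# A large, offset capsule on dyn_macaron_01 forms the main body, while a tiny capsule on dyn_macaron_MSTR provides the anchor body.
# A single constraint links the two bones with equal swing and twist limit angles, which suggests a dangling, swaying ornament.
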PhysicsAsset /VehicleCosmetics/PassthroughMeshes/Macaron/SK_Macaron_PhysicsAsset_MutableContainer
bodies {
 dyn_macaron_01 [{shapeType=Capsule, length=31.699, radius=31.697, center=<18.9431 0 -0.143>, rotation=<rot 89.3547 0.0059 90.0059>}]
 dyn_macaron_MSTR [{shapeType=Capsule, length=0.505, radius=0.505, rotation=<rot 0.0001 -90 0>}]
}
constraints [{bone1Name=dyn_macaron_01, bone2Name=dyn_macaron_MSTR, swing1LimitDegrees=45, swing2LimitDegrees=45, twistLimitDegrees=45}]
